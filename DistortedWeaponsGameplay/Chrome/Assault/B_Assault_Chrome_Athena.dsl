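; Chrome assault rifle weapon Blueprint that drives crystal visuals for XP progress, rarity upgrades and shattering.
; On BeginPlay it plays a local-only crystal reform gameplay cue and binds to the Chrome weapon XP component.
; It writes XP percentage, rarity tier, a legendary-or-higher flag and animation start times into the weapon mesh's custom primitive data, avoiding dynamic material instances.
; An AnimNotify stamps the shatter time during the upgrade animation.
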(event EventBeginPlay
  (bind _returnvalue (Actor|GetComponentbyClass "/DistortedWeaponsGameplay/Chrome/C_ChromeWeapons_XPComponent.C_ChromeWeapons_XPComponent_C"))
  ;[0.08891 1 0.1035] Bind events for handling fire mode data swaps
  (|CallParentFunction)
  (bind _asfort_player_pawn (Utilities|Casting|CastToFortPlayerPawn (Game|GetInstigator))
    (:then
      (if (Pawn|IsLocallyControlled _asfort_player_pawn)
        (GameplayCue|ExecuteGameplayCue%28Burst%29Local (Game|GetInstigator) (Variables|Default|GetGCN_ReformCrystals) (Ability|GameplayCue|MakeGameplayCueParameters 0.0))
        (else
          _))
      (CallFunction|MIDUpgradeAnimation)
      (if (Utilities|IsValid _returnvalue :!class /Script/Engine.KismetSystemLibrary)
        (Variables|Default|SetXPComponentRef _returnvalue)
        (Default|BindEventtoOnXPChanged _returnvalue (EventDispatchers|CreateEvent :!event Custom|On%20XP%20Change))))
    (:CastFailed
      _)))

(event Custom|AnimNotify_Upgrade_ShatterCrystals
  ;[1 1 1] Event that plays when crystals shatter in upgrade animation
  (CallFunction|MIDShatterCrystals))

(event Custom|On%20XP%20Change
  (params (XPDelta :type Float%20%28double-precision%29) (CurrentXPPercentage :type Float%20%28double-precision%29))
  ;[1 1 1] Event that plays when crystals shatter in upgrade animation
  (CallFunction|MIDChangeXP CurrentXPPercentage))

(fn MIDChangeXP
  (params (XP%20Normalized :type Float%20%28double-precision%29))
  (Rendering|Material|SetCustomPrimitiveDataFloat (Variables|Visual|GetWeaponMesh) 0 XP%20Normalized))

(fn MIDShatterCrystals
  (Rendering|Material|SetCustomPrimitiveDataFloat (Variables|Visual|GetWeaponMesh) 4 (Utilities|Time|GetTimeSeconds)))

(fn MIDUpgradeAnimation
  (bind _returnvalue (Rarity|GetWeaponDataRarity))
  (bind _asfort_player_pawn (Utilities|Casting|CastToFortPlayerPawn (Game|GetInstigator))
    (:then
      (if (Pawn|IsLocallyControlled _asfort_player_pawn)
        (Rendering|Material|SetCustomPrimitiveDataFloat (Variables|Visual|GetWeaponMesh) 2 (Utilities|Time|GetTimeSeconds))
        (else
          _))
      (Rendering|Material|SetCustomPrimitiveDataFloat (Variables|Visual|GetWeaponMesh) 1 (Utilities|Select :Index _returnvalue :Common 0.0 :Uncommon 0.0 :Rare 1.0 :Epic 2.0 :Legendary 2.0 :Mythic 2.0 :Transcendent 2.0 :Unattainable 2.0))
      (Rendering|Material|SetCustomPrimitiveDataFloat (Variables|Visual|GetWeaponMesh) 3 (Math|Conversions|ToFloat(Boolean) (Utilities|Select :Index _returnvalue :Common false :Uncommon false :Rare false :Epic false :Legendary true :Mythic true :Transcendent true :Unattainable true))))
    (:CastFailed
      _)))

(fn ConstructionScript
  (|CallParentFunction))
