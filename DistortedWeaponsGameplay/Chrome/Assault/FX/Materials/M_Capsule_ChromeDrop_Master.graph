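# Masked, lit master material for Niagara sprite chrome droplets, stretched into teardrops by the engine Sprite_TearDrop function.
# A per-particle randomized, rotating and panning flowmap distorts the teardrop UVs, which drive the LackLuster base function and a soft sphere mask for particle-colored emissive and alpha.
# A PreviousFrameSwitch offsets last-frame position by particle velocity times delta time.
# The MaxStretch dynamic parameter also feeds both speed inputs, leaving MinSpeed and MaxSpeed unused.
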Material /DistortedWeaponsGameplay/Chrome/Assault/FX/Materials/M_Capsule_ChromeDrop_Master
graph {
 properties {
  bEnableResponsiveAA=true
  bTangentSpaceNormal=false
  bUsedWithNiagaraSprites=true
  blendMode=BLEND_Masked
  opacityMaskClipValue=0.1
  pixelDepthOffsetMode=PDOM_Legacy
  usedShadingModels=Default Lit
 }
 nodes [
  n1 = DynamicParameter() {defaultValue=<rgba 5 10 100 3>, paramNames [Distort,MinSpeed,MaxSpeed,MaxStretch]}
  n2 = Time()
  n3 = ScalarParameter() {defaultValue=1.5, parameterName=PanSpeed}
  n4 = ParticleRandom()
  n5 = Round(None=@n4)
  n6 = Add(A=Multiply(A=Multiply(A=@n2, B=@n3), B=LinearInterpolate(Alpha=@n5) {constA=-1}), B=@n4)
  n7 = Rotator(Coordinate=Multiply(A=ScalarParameter() {defaultValue=0.5, parameterName=DistortionTile}, B=TextureCoordinate()), Time=@n6) {centerX=0.25, centerY=0.25, speed=3}
  n8 = ComponentMask(None=TextureSample(UVs=Panner(Coordinate=@n7, Time=@n6) {speedX=0.1, speedY=-1}) {samplerType=SAMPLERTYPE_LinearColor, texture=/Game/Environments/Apollo/Sets/Jerky/Materials/T_JerkyFlowmap}.RGB) {g=true, r=true}
  n9 = MaterialFunctionCall!(
   Max Stretch Factor=@n1.MaxStretch
   Min Tear Drop Width=ScalarParameter() {defaultValue=0.25, parameterName=MinPointWidth}
   Minimum Speed=@n1.MaxStretch
   Max Speed=@n1.MaxStretch
   ------ Particle Defaults Below ------
   Particle Speed
   Particle Direction
   Particle Position
   Particle Size
   UV Add=Multiply(A=Multiply(A=@n1.Distort) {constB=0.125}, B=ConstantBiasScale(None=@n8) {bias=-0.5, scale=2}) {constB=0.2}
  ) {materialFunction=/Engine/Functions/Engine_MaterialFunctions02/WorldPositionOffset/Sprite_TearDrop}
  BreakMaterialAttributes!(Attr=MaterialFunctionCall!(
    bBlendMaterialAttributes
    MA_In
    UVsOverride=@n9.Texture Coordinates
    Override Normals
    BlendMaterialAttributesAlpha
    BaseColorOverride
    WorldPositionOverride
    VertexNormalOverride
    SourceNormalsStrengthOverride
   ) {materialFunction=/Game/TechArt/Seasonal/LackLuster/MaterialFunctions/MF_LackLuster_Base}.MA Out)
  n10 = Constant() {r=0.5}
  n11 = SphereMask(A=@n9.Texture Coordinates, B=@n10, Radius=@n10) {hardnessPercent=0}
  n12 = ScalarParameter() {defaultValue=4, parameterName=EmissivePower}
  n13 = ParticleColor()
  MaterialFunctionCall(Emissive=Multiply(A=Power(Base=@n11, Exp=@n12), B=@n13.RGB)) {materialFunction=&MF_FX_EmissiveAdjustments}
  Multiply(A=@n13.A, B=@n11)
  n14 = Multiply(A=Multiply(A=ParticleDirection(), B=ParticleSpeed()), B=DeltaTime())
  PreviousFrameSwitch(Current Frame=@n9.World Position Offset, Previous Frame=Subtract(A=@n9.World Position Offset, B=Multiply(A=@n14)))
 ]
}
