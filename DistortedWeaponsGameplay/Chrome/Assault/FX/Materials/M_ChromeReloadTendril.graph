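# Translucent, unlit Niagara mesh-particle material for the Chrome assault weapon's reload tendril effect.
# Timer-panned noise layers distort a sphere-mask strand that blends cyan to green with hue shift, while Fadeout erodes it along V.
# SplineThicken sets width, and a clamped offset from particle position to a Niagara-driven value bends it progressively along V.
# Opacity also fades by view angle against the thickened normal, depth fade and GlobalAlpha.
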Material /DistortedWeaponsGameplay/Chrome/Assault/FX/Materials/M_ChromeReloadTendril
graph {
 properties {
  bUsedWithNiagaraMeshParticles=true
  blendMode=BLEND_Translucent
  pixelDepthOffsetMode=PDOM_Legacy
  shadingModel=MSM_Unlit
  twoSided=true
  usedShadingModels=Unlit
 }
 nodes [
  n1 = TextureObjectParameter() {
   group=NoiseA
   parameterName=NoiseA_Texture
   samplerType=SAMPLERTYPE_LinearGrayscale
   sortPriority=0
   texture=/Game/Gadgets/Assets/VinderTech_GliderChute/Glider_Thrive_Female/FX/Texture/T_AntiMaterialCharge_Dissolve_01
  }
  n2 = VectorParameter() {group=NoiseA, parameterName=NoiseA_UVControl, channelNames {a=VPan, b=UPan, g=VScale, r=UScale}, defaultValue=<rgba 1 0.5 -0.2 1>}
  n3 = DynamicParameter() {paramNames [Timer, Fadeout, RandomTimeOffset, Param4]}
  n4 = Add(A=@n3.Timer, B=@n3.Timer)
  n5 = MaterialFunctionCall(Texture=@n1, UVControl_RGB=@n2, UVControl_A=@n2, "[Override]TimeInput"=@n4) {materialFunction=&FXCommon_TextureSample}
  n6 = VectorParameter() {
   group=NoiseA
   parameterName=NoiseA_NoiseControl
   sortPriority=60
   channelNames {a=VStrength, b=UStrength, g=VOffset, r=UOffset}
   defaultValue=<rgba 0.5 0 0 0.5>
  }
  n7 = SphereMask(
   A=Add(
    A=TextureCoordinate() {vTiling=0}
    B=MaterialFunctionCall!(NoiseValue=@n5.RG, Control_RGB=@n6, Control_A=@n6) {materialFunction=/Game/Effects/Niagara/MasterMaterials/Functions/Common/FXCommon_NoiseControl}.NoiseValue
   )
   B=Constant2Vector() {r=0.5}
   Radius=Constant() {r=0.5}
   Hardness=Constant()
  )
  n8 = TextureObjectParameter() {
   group=NoiseA
   parameterName=NoiseB_Texture
   samplerType=SAMPLERTYPE_LinearGrayscale
   sortPriority=0
   texture=&T_SimplePerlin_Greyscale
  }
  n9 = VectorParameter() {group=NoiseA, parameterName=NoiseB_UVControl, channelNames {a=VPan, b=UPan, g=VScale, r=UScale}, defaultValue=<rgba 1.5 1 0.5 2>}
  n10 = VectorParameter() {
   group=NoiseA
   parameterName=NoiseB_NoiseControl
   sortPriority=60
   channelNames {a=VStrength, b=UStrength, g=VOffset, r=UOffset}
   defaultValue=<rgba 0.5 0.5 0.5 0.5>
  }
  n11 = Add(
   A=@n7
   B=MaterialFunctionCall(
    NoiseValue=MaterialFunctionCall(Texture=@n8, UVControl_RGB=@n9, UVControl_A=@n9, "[Override]TimeInput"=@n4) {materialFunction=&FXCommon_TextureSample}.RG
    Control_RGB=@n10
    Control_A=@n10
   ) {materialFunction=&FXCommon_NoiseControl}.NoiseValue
  )
  n12 = TextureObjectParameter() {
   group=NoiseA
   parameterName=OpacOverlay_Texture
   sortPriority=0
   texture=/Game/Effects/Fort_Effects/Emotes/UltraLight/Texture/T_SoftNoise_BC7
  }
  n13 = VectorParameter() {group=NoiseA, parameterName=OpacOverlay_UVControl, channelNames {a=VPan, b=UPan, g=VScale, r=UScale}, defaultValue=<rgba 1 1 0.5 2>}
  n14 = VectorParameter() {
   group=NoiseA
   parameterName=OpacOverlay_NoiseControl
   sortPriority=60
   channelNames {a=VStrength, b=UStrength, g=VOffset, r=UOffset}
   defaultValue=<rgba 0.4 0.5 0.2 0.3>
  }
  n15 = MaterialFunctionCall(
   NoiseValue=MaterialFunctionCall(Texture=@n12, UVControl_RGB=@n13, UVControl_A=@n13, "[Override]TimeInput"=@n4) {materialFunction=&FXCommon_TextureSample}.R
   Control_RGB=@n14
   Control_A=@n14
  ) {materialFunction=&FXCommon_NoiseControl}
  n16 = Multiply(A=@n15.NoiseValue) {constB=0.5}
  n17 = Constant3Vector() {constant=<rgba 0 0.9455 1 0>}
  n18 = Constant3Vector() {constant=<rgba 0 1 0.1443 0>}
  n19 = Add(A=@n3.Fadeout) {constB=0.5}
  n20 = OneMinus(None=ComponentMask(None=TextureCoordinate()) {g=true})
  n21 = ComponentMask(None=TextureCoordinate()) {g=true}
  n22 = Multiply(A=@n7, B=Multiply(A=SmoothStep(Min=@n3.Fadeout, Max=@n19, Value=@n20), B=SmoothStep(Value=@n21) {constMax=0.25}))
  MaterialFunctionCall(Emissive=MaterialFunctionCall(Hue Shift Percentage=Add(A=@n11, B=@n16), Texture=LinearInterpolate(A=@n17, B=@n18, Alpha=@n22)) {materialFunction=&HueShift}.Result) {materialFunction=&MF_FX_EmissiveAdjustments}
  n23 = Multiply(A=@n22, B=Add(A=@n15.NoiseValue) {constB=0.2})
  n24 = CameraVectorWS()
  n25 = LinearInterpolate(
   B=DynamicParameter() {parameterIndex=1, paramNames [SplineWidth,GlobalAlpha,Param3,Param4]}.SplineWidth
   Alpha=Saturate(None=Add(A=@n21) {constB=0.25})
  )
  n26 = DynamicParameter() {desc=Parameters set by Niagara, parameterIndex=2, defaultValue=<rgba 1 1 50 1>, paramNames [WPO UV Scale, "", "", ""]}
  n27 = MaterialFunctionCall(X=@n26.WPO UV Scale, Y=@n26.WPO UV Scale, Z=@n26.WPO UV Scale) {materialFunction=&MakeFloat3}
  n28 = ParticlePositionWS()
  n29 = ScalarParameter() {defaultValue=100, group=WPO, parameterName=Interpolate Clamp}
  n30 = TextureCoordinate()
  n31 = ComponentMask(None=@n30) {g=true}
  n32 = ScalarParameter() {defaultValue=2, group=WPO, parameterName=WPOBendPower}
  n33 = LinearInterpolate(
   B=Clamp(None=Subtract(A=@n27.Result, B=@n28), Min=Multiply(B=@n29) {constA=-1}, Max=@n29) {maxDefault=1000, minDefault=-1000}
   Alpha=Multiply(A=Power(Base=@n31, Exp=@n32), B=@n26.A)
  )
  n34 = MaterialFunctionCall(WidthBase=@n25, WidthTip=@n25, WorldPosition=Add(A=@n33, B=WorldPosition() {worldPositionShaderOffset=WPT_ExcludeAllShaderOffsets}.XYZ)) {materialFunction=&SplineThicken}
  n35 = Multiply(A=Abs(None=DotProduct(A=@n24, B=@n34.Normal)), B=ScalarParameter() {defaultValue=1.3, parameterName=FadeSplineThicken_Mul})
  n36 = MaterialFunctionCall(
   Input=Power(Base=Saturate(None=@n35), Exp=ScalarParameter() {defaultValue=2, parameterName=FadeSplineThicken_Power})
   Input Low=Constant()
   Input High=Constant() {r=1}
  ) {materialFunction=&RemapValue_0-1}
  Multiply(
   A=Multiply(A=Multiply(A=@n23, B=@n36.Result), B=DepthFade() {fadeDistanceDefault=25})
   B=DynamicParameter() {parameterIndex=1, paramNames [SplineWidth,GlobalAlpha,Param3,Param4]}.GlobalAlpha
  ) {constB=0.3}
  OneMinus(None=@n30)
  Reroute(None=Add(A=@n33, B=@n34.WorldPositionOffset))
 ]
}
