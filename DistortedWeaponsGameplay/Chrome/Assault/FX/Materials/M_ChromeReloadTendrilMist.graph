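# Translucent unlit Niagara mesh-particle material for misty tendrils in the Chrome assault weapon's reload effect.
# World position offset bends the mesh along V toward a Niagara-supplied position, with SplineThicken setting tapered width.
# Three panning samples of a chrome sand texture, timed by particle parameters, drive hue-shifted emissive and opacity.
# A noise-distorted edge mask, Fadeout wipe and view-angle falloff soften the tendrils.
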Material /DistortedWeaponsGameplay/Chrome/Assault/FX/Materials/M_ChromeReloadTendrilMist
graph {
 properties {
  bUsedWithNiagaraMeshParticles=true
  blendMode=BLEND_Translucent
  pixelDepthOffsetMode=PDOM_Legacy
  shadingModel=MSM_Unlit
  twoSided=true
  usedShadingModels=Unlit
 }
 nodes [
  n1 = ComponentMask(None=TextureCoordinate()) {g=true}
  n2 = LinearInterpolate(
   B=DynamicParameter() {parameterIndex=1, paramNames [SplineWidth,Param2,Param3,Param4]}.SplineWidth
   Alpha=Saturate(None=Add(A=@n1) {constB=0.25})
  )
  n3 = DynamicParameter() {desc=Parameters set by Niagara, parameterIndex=2, defaultValue=<rgba 1 1 50 1>, paramNames [WPO UV Scale, "", "", ""]}
  n4 = MaterialFunctionCall(X=@n3.WPO UV Scale, Y=@n3.WPO UV Scale, Z=@n3.WPO UV Scale) {materialFunction=&MakeFloat3}
  n5 = ParticlePositionWS()
  n6 = ScalarParameter() {defaultValue=100, group=WPO, parameterName=Interpolate Clamp}
  n7 = TextureCoordinate()
  n8 = ComponentMask(None=@n7) {g=true}
  n9 = ScalarParameter() {defaultValue=2, group=WPO, parameterName=WPOBendPower}
  n10 = LinearInterpolate(
   B=Clamp(None=Subtract(A=@n4.Result, B=@n5), Min=Multiply(B=@n6) {constA=-1}, Max=@n6) {maxDefault=1000, minDefault=-1000}
   Alpha=Multiply(A=Power(Base=@n8, Exp=@n9), B=@n3.A)
  )
  n11 = MaterialFunctionCall(WidthBase=@n2, WidthTip=@n2, WorldPosition=Add(A=@n10, B=WorldPosition() {worldPositionShaderOffset=WPT_ExcludeAllShaderOffsets}.XYZ)) {materialFunction=&SplineThicken}
  Add(A=@n11.WorldPositionOffset, B=@n10)
  n12 = TextureObjectParameter() {
   group=Main
   parameterName=Main_Texture
   samplerType=SAMPLERTYPE_Normal
   sortPriority=0
   texture=/DistortedWeaponsGameplay/Chrome/Assault/FX/Textures/T_Chrome_Sand_Weapons
  }
  n13 = VectorParameter() {group=Main, parameterName=Main_UVControl, channelNames {a=VPan, b=UPan, g=VScale, r=UScale}, defaultValue=<rgba 0.2 0.5 -0.05 0.1>}
  n14 = NamedRerouteUsage()
  n15 = MaterialFunctionCall(Texture=@n12, UVControl_RGB=@n13, UVControl_A=@n13, "[Override]TimeInput"=@n14) {materialFunction=&FXCommon_TextureSample}
  n16 = DynamicParameter() {defaultValue=<rgba 1 0 1 1>, paramNames [Timer, Fadeout, RandomTimeOffset, Param4]}
  n17 = TextureCoordinate()
  n18 = SmoothStep(Min=@n16.Fadeout, Max=Add(A=@n16.Fadeout) {constB=0.5}, Value=OneMinus(None=ComponentMask(None=@n17) {g=true}))
  n19 = TextureObjectParameter() {
   group=NoiseA
   parameterName=NoiseA_Texture
   samplerType=SAMPLERTYPE_LinearGrayscale
   sortPriority=0
   texture=/Game/Gadgets/Assets/VinderTech_GliderChute/Glider_Thrive_Female/FX/Texture/T_AntiMaterialCharge_Dissolve_01
  }
  n20 = VectorParameter() {group=NoiseA, parameterName=NoiseA_UVControl, channelNames {a=VPan, b=UPan, g=VScale, r=UScale}, defaultValue=<rgba 1 0.5 0 0.5>}
  n21 = MaterialFunctionCall(Texture=@n19, UVControl_RGB=@n20, UVControl_A=@n20, "[Override]TimeInput"=NamedRerouteUsage()) {materialFunction=&FXCommon_TextureSample}
  n22 = VectorParameter() {
   group=NoiseA
   parameterName=NoiseA_NoiseControl
   sortPriority=60
   channelNames {a=VStrength, b=UStrength, g=VOffset, r=UOffset}
   defaultValue=<rgba 0.5 0 0 0.5>
  }
  n23 = SphereMask(
   A=Add(
    A=TextureCoordinate() {vTiling=0}
    B=MaterialFunctionCall!(NoiseValue=@n21.RG, Control_RGB=@n22, Control_A=@n22) {materialFunction=/Game/Effects/Niagara/MasterMaterials/Functions/Common/FXCommon_NoiseControl}.NoiseValue
   )
   B=Constant2Vector() {r=0.5}
   Radius=Constant() {r=0.5}
   Hardness=Constant()
  )
  n24 = TextureObjectParameter() {
   group=Tri
   parameterName=Tri_Texture
   samplerType=SAMPLERTYPE_Normal
   sortPriority=0
   texture=&T_Chrome_Sand_Weapons
  }
  n25 = VectorParameter() {group=Tri, parameterName=Tri_UVControl, channelNames {a=VPan, b=UPan, g=VScale, r=UScale}, defaultValue=<rgba 0.15 0.5 0.1 0.2>}
  n26 = Multiply(A=Add(A=MaterialFunctionCall(Texture=@n24, UVControl_RGB=@n25, UVControl_A=@n25, "[Override]TimeInput"=@n14) {materialFunction=&FXCommon_TextureSample}.G)) {constB=0.5}
  n27 = ScalarParameter() {parameterName=TriMask2}
  n28 = Multiply(
   A=Power(Base=Multiply(A=@n18, B=@n23), Exp=ScalarParameter() {defaultValue=1, parameterName=TriMask})
   B=MaterialFunctionCall(In=Power(Base=@n26, Exp=@n27), Contrast=ScalarParameter() {parameterName="TriMask2@"}) {materialFunction=&CheapContrast}.Result
  )
  n29 = Add(
   A=Multiply(A=@n28, B=VectorParameter() {parameterName=TriC, defaultValue=<rgba 0.2554 0.4623 0.776 1>}.RGB)
   B=VectorParameter() {parameterName=Color, defaultValue=<rgba 100 0 0 1>}.RGB
  )
  n30 = TextureObjectParameter() {
   group=Sec
   parameterName=Sec_Texture
   samplerType=SAMPLERTYPE_Normal
   sortPriority=0
   texture=&T_Chrome_Sand_Weapons
  }
  n31 = VectorParameter() {group=Sec, parameterName=Sec_UVControl, channelNames {a=VPan, b=UPan, g=VScale, r=UScale}, defaultValue=<rgba 0.1 0.5 0.1 0.3>}
  n32 = Add(A=MaterialFunctionCall(Texture=@n30, UVControl_RGB=@n31, UVControl_A=@n31, "[Override]TimeInput"=@n14) {materialFunction=&FXCommon_TextureSample}.G)
  n33 = Multiply(A=Multiply(A=Add(A=@n15.G)) {constB=0.5}, B=Add(A=Multiply(A=@n32) {constB=0.5}, B=ScalarParameter() {parameterName=aw}))
  n34 = ParticleColor()
  n35 = Multiply(
   A=Multiply(
    A=MaterialFunctionCall(Hue Shift Percentage=@n15.R, Texture=@n29) {materialFunction=&HueShift}.Result
    B=@n33
   )
   B=@n34.RGB
  )
  MaterialFunctionCall(Emissive=@n35) {materialFunction=&MF_FX_EmissiveAdjustments}
  n36 = DotProduct(A=CameraVectorWS(), B=@n11.Normal)
  n37 = Saturate(None=Multiply(A=Abs(None=@n36), B=ScalarParameter() {defaultValue=1.3, parameterName=FadeSplineThicken_Mul}))
  n38 = ScalarParameter() {defaultValue=2, parameterName=FadeSplineThicken_Power}
  n39 = Saturate(None=@n33)
  n40 = Multiply(A=Add(A=Multiply(A=@n39, B=@n18), B=ScalarParameter()), B=ScalarParameter() {defaultValue=0.5, parameterName=TotalOpacityScale})
  n41 = Multiply(A=@n28, B=ScalarParameter() {defaultValue=1, parameterName=TriToOpacityStrength})
  Multiply(
   A=Multiply(A=@n34.A, B=Multiply(A=@n40, B=@n41))
   B=MaterialFunctionCall(Input=Power(Base=@n37, Exp=@n38), Input Low=Constant(), Input High=Constant() {r=1}) {materialFunction=&RemapValue_0-1}.Result
  )
  NamedRerouteDeclaration(None=Add(A=@n16.Timer, B=@n16.Timer)) {name=Time, nodeColor=<rgba 0.875 1 0 1>}
  OneMinus(None=@n7)
  Reroute(None=@n35)
 ]
}
