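# Translucent, unlit, two-sided Niagara sprite material for a colored smoke overlay in the Chrome assault weapon effects.
# A 4x4 motion-vector flipbook, driven by the Animation dynamic parameter, offsets the UVs of panning Perlin noise.
# Noise red indexes a curve-atlas gradient tinted by particle color.
# Noise green, flipbook and particle alpha multiply into opacity, with depth fade replaced by a zero constant on mobile.
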Material /DistortedWeaponsGameplay/Chrome/Assault/FX/Materials/M_ChromeSmokeColorOverlay
graph {
 properties {
  bUsedWithNiagaraSprites=true
  blendMode=BLEND_Translucent
  pixelDepthOffsetMode=PDOM_Legacy
  shadingModel=MSM_Unlit
  twoSided=true
  usedShadingModels=Unlit
 }
 nodes [
  n1 = TextureCoordinate()
  n2 = Constant() {r=4}
  n3 = TextureObjectParameter() {parameterName=MotionVector, samplerType=SAMPLERTYPE_LinearColor, texture=/Game/Effects/Niagara/MasterMaterials/Smoke/Textures/Vector_OF}
  n4 = MaterialFunctionCall!(
   "Animation  Phase (0-1)"=DynamicParameter() {paramNames [Animation, Unused, Emissive, NormalMultiplier]}.Animation
   Number of Columns X=@n2
   Number of Rows Y=@n2
   Texture=TextureObject() {texture=/Game/Effects/Niagara/MasterMaterials/Smoke/Textures/NEW/RT4x4_4_Tex}
   UVs
   Motion Vector Texture=@n3
   Phase to Debug
   Motion Vector Intensity=ScalarParameter() {defaultValue=0.344, group=Motion Vectors, parameterName=MotionVector}
   Double Apply Motion Vectors
   Custom UV Input
   Custom UV2 Input
   Use Custom UVs
  ) {materialFunction=/Engine/Functions/Engine_MaterialFunctions02/Texturing/FlipBook_MotionVectors}
  n5 = TextureSample(UVs=Add(A=Panner(Coordinate=@n1) {speedX=0.25, speedY=0.25}, B=ComponentMask(None=@n4.Result) {r=true})) {texture=/Game/Global/Textures/Default/Noise/3xPerlin_M}
  n6 = CurveAtlasRowParameter(CurveTime=Power(Base=Multiply(A=@n5.R) {constB=2})) {atlas=&AthenaCurveAtlas01, curve=/Game/Athena/LinearColorCurves/Curve_FritterEvent, defaultValue=212}
  n7 = ParticleColor()
  n8 = Multiply(A=@n6, B=@n7.RGB)
  MaterialFunctionCall(Emissive=@n8) {materialFunction=&MF_FX_EmissiveAdjustments}
  n9 = DepthFade() {opacityDefault=25}
  n10 = Constant()
  Multiply(
   A=@n7.A
   B=Multiply(A=Multiply(A=@n5.G, B=@n4.Result), B=ShadingPathSwitch(Default=@n9, Deferred=@n9, Mobile=@n10) {desc=Migrated from FeatureLevelSwitch})
  )
  Reroute(None=@n8)
 ]
}
