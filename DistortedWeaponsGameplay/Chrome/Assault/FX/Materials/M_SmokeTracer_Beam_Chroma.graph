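# An additive, unlit, two-sided Niagara material for a chroma-tinted smoke tracer beam in Chrome assault weapon effects.
# A masked dim-to-bright gradient, hue-shifted by noise, combines with smoke and particle color for emissive.
# Two panned smoke trail samples form eroded opacity, faded by a length-scaled UV mask, particle alpha and camera proximity.
# SplineThicken sets beam width, while dynamic parameters supply erosion, pan, gradient, length, width and random time offset.
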Material /DistortedWeaponsGameplay/Chrome/Assault/FX/Materials/M_SmokeTracer_Beam_Chroma
graph {
 properties {
  bUsedWithNiagaraMeshParticles=true
  bUsedWithNiagaraSprites=true
  blendMode=BLEND_Additive
  pixelDepthOffsetMode=PDOM_Legacy
  shadingModel=MSM_Unlit
  translucencyLightingMode=TLM_VolumetricPerVertexNonDirectional
  twoSided=true
  usedShadingModels=Unlit
 }
 nodes [
  MaterialFunctionCall() {materialFunction=&DebugTimeSine}
  n1 = ComponentMask(None=TextureCoordinate()) {g=true}
  n2 = Constant() {r=1}
  n3 = NamedRerouteUsage()
  n4 = Multiply(A=@n3) {constA=1, constB=1.1}
  n5 = Saturate(None=LinearInterpolate(Alpha=@n3) {constA=-1, constB=0.9})
  n6 = Multiply(
   A=Power(Base=SphereMask(A=@n1, B=@n2, Radius=@n4, Hardness=@n5) {hardnessPercent=5}, Exp=LinearInterpolate(Alpha=@n3) {constA=3})
   B=LinearInterpolate(Alpha=@n3) {constA=-0.1, constB=2}
  )
  n7 = MaterialFunctionCall(
   Hue Shift Percentage=TextureSample(UVs=NamedRerouteUsage()) {texture=/Game/Global/Textures/Default/Noise/3xPerlin_M}.RGB
   Texture=LinearInterpolate(
    A=VectorParameter() {parameterName=DimColor, defaultValue=<rgba 0.125 0.4017 0.5 1>}.RGB
    B=VectorParameter() {parameterName=BrightColor, defaultValue=<rgba 0.5547 1.5 0.6675 0>}.RGB
    Alpha=Clamp(None=@n6)
   )
  ) {materialFunction=&HueShift}
  n8 = DynamicParameter() {paramNames [Erode, Pan, Gradient 1-0, Length]}
  n9 = Time()
  n10 = Add(
   A=@n8.Pan
   B=Add(A=Multiply(A=@n9) {constB=-0.25}, B=DynamicParameter() {parameterIndex=1, paramNames [WidthMult,Random,Param3,Param4]}.Random)
  )
  n11 = VectorParameter() {parameterName=SmokeSpeed, defaultValue=<rgba -0.1 0 0 1>}
  n12 = TextureSampleParameter2D(UVs=Panner(Coordinate=NamedRerouteUsage(), Time=@n10, Speed=@n11.RGB) {speedX=0.1}) {parameterName=TrailTexture, samplerType=SAMPLERTYPE_Grayscale, texture=&T_Smoke_Trail_Soft}
  n13 = NamedRerouteUsage()
  n14 = Multiply(A=@n11.RGB) {constB=0.5}
  n15 = TextureSampleParameter2D(UVs=Multiply(A=Panner(Coordinate=@n13, Time=@n10, Speed=@n14) {speedX=0.1}, B=Constant2Vector() {g=1.08, r=1.5}) {constB=2}) {parameterName=TrailTexture, samplerType=SAMPLERTYPE_Grayscale, texture=&T_Smoke_Trail_Soft}
  n16 = LinearInterpolate(A=@n12.G, B=@n15.G) {constAlpha=0.37}
  n17 = Multiply(A=@n7.Result, B=@n16)
  n18 = ScalarParameter() {defaultValue=0.2, parameterName=SmokeBrightness}
  n19 = MaterialFunctionCall(LDR Input=Clamp(None=@n7.Result), HDR Tint=@n7.Result) {materialFunction=&DeriveHDRfromLDR}
  n20 = ParticleColor() {desc="R = erosion, G = panning, B = emissive"}
  MaterialFunctionCall(Emissive=Add(A=Multiply(A=@n17, B=@n18) {constB=0.07}, B=Multiply(A=@n19.RemappedRange, B=@n20.RGB))) {materialFunction=&MF_FX_EmissiveAdjustments}
  n21 = Multiply(
   A=ScalarParameter() {defaultValue=50, parameterName=SplineWidth}
   B=DynamicParameter() {parameterIndex=1, paramNames [WidthMult,Random,Param3,Param4]}.WidthMult
  )
  MaterialFunctionCall(WidthBase=Multiply(A=@n21, B=ScalarParameter() {defaultValue=1, parameterName=WidthEndMult}), WidthTip=@n21) {materialFunction=&SplineThicken}
  n22 = ComponentMask(None=TextureCoordinate()) {g=true}
  n23 = OneMinus(None=@n22)
  n24 = Clamp(None=LinearInterpolate(B=Multiply(A=@n8.Length) {constB=50}, Alpha=Multiply(A=@n23, B=@n22)) {constB=40})
  n25 = NamedRerouteUsage()
  n26 = Add(A=@n25, B=@n16)
  n27 = Multiply(A=Multiply(A=LinearInterpolate(Alpha=@n26) {constA=-3.5, constB=1.25}, B=@n25), B=@n16)
  n28 = LinearInterpolate(Alpha=@n25) {constA=0.5, constB=3}
  n29 = Multiply(A=@n24, B=Clamp(None=Power(Base=@n27, Exp=@n28)))
  Multiply(
   A=Multiply(A=@n20.A, B=Multiply(A=@n24, B=@n29))
   B=MaterialFunctionCall(Distance=ScalarParameter() {defaultValue=400, parameterName=NearCameraFadeDistance}) {materialFunction=&Near_Camera_Fade}.Result
  )
  NamedRerouteDeclaration(None=@n8.Gradient 1-0) {name=GradientParam, nodeColor=<rgba 0.3594 1 0 1>}
  NamedRerouteDeclaration(None=@n8.Erode) {name=ErodeParam, nodeColor=<rgba 0 1 0.6484 1>}
  n30 = Add(A=TextureCoordinate() {coordinateIndex=1, vTiling=0.25}, B=Constant2Vector() {g=0.33})
  n31 = Add(A=TextureCoordinate() {coordinateIndex=1, vTiling=2.5}, B=Constant2Vector() {g=-0.75})
  n32 = NamedRerouteUsage()
  n33 = Divide(A=@n8.Length, B=ScalarParameter() {defaultValue=1.5, parameterName=SmokeTileDistance})
  n34 = Constant() {r=1}
  NamedRerouteDeclaration(None=Multiply(A=LinearInterpolate(A=@n30, B=@n31, Alpha=@n32), B=AppendVector(A=@n33, B=@n34))) {name=UV, nodeColor=<rgba 0.7578 0 1 1>}
 ]
}
