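# Material function for Chrome assault weapon crystals that grow, shake, break and glow, applied when Use Crystal Growth is on.
# Pivot Painter pivots and random values drive per-crystal world position offset, with growth spreading from an origin point; particle variants skip the offset.
# Custom primitive data supplies EXP, growth level and animation and break times.
# It restyles MF_LackLuster_Petrified output with desaturated color, input normals and EXP-driven emissive using a panning, bump-offset texture.
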MaterialFunction /DistortedWeaponsGameplay/Chrome/Assault/Materials/MF_Chrome_Weapon_Crystals
graph {nodes [
  n1 = StaticSwitchParameter(True=Constant(), False=Add(A=NamedRerouteUsage(), B=NamedRerouteUsage())) {parameterName=Is Particle}
  n2 = GetMaterialAttributes(None=MaterialFunctionCall!(
    bBlendMaterialAttributes
    MA_In
    BlendMaterialAttributesAlpha
    BaseColorOverride
    VertexNormalOverride
    Texture Override Position=NamedRerouteUsage()
   ) {materialFunction=/Game/TechArt/Seasonal/LackLuster/MaterialFunctions/MF_LackLuster_Petrified}.MA Out) {attributeGetTypes [#1, #2, #3]}
  n3 = ParticleColor()
  n4 = ScalarParameter() {bUseCustomPrimitiveData=true, parameterName=EXP}
  n5 = Multiply(
   A=@n2.EmissiveColor
   B=LinearInterpolate(Alpha=StaticSwitchParameter(True=@n3.A, False=@n4) {parameterName=Is Particle}) {constA=1, constB=2}
  )
  n6 = NamedRerouteUsage()
  n7 = FunctionInput() {
   bUsePreviewValueAsDefault=true
   inputName=Upgrade Emissive Multiplier
   inputType=FunctionInput_Scalar
   sortPriority=1000
   previewValue=<1 0 0 1>
  }
  n8 = Desaturation(None=@n2.BaseColor, Fraction=ScalarParameter() {defaultValue=0.8, parameterName=Diffuse Desat})
  n9 = ScalarParameter() {defaultValue=3, parameterName=Diffuse Lift}
  n10 = FunctionInput() {inputType=FunctionInput_MaterialAttributes}
  n11 = SetMaterialAttributes!(
   MaterialAttributes=@n2.MaterialAttributes
   Normal=GetMaterialAttributes(None=@n10) {attributeGetTypes [#1, #2, #3]}.Normal
   EmissiveColor=Multiply(A=Add(A=@n5, B=@n6), B=@n7)
   BaseColor=LinearInterpolate(A=@n2.BaseColor, B=Multiply(A=@n8, B=@n9), Alpha=ScalarParameter() {defaultValue=1, parameterName=bleh})
  ) {attributeSetTypes [#1, #2, #3]}
  FunctionOutput(None=StaticSwitchParameter(
    True=MaterialFunctionCall!(Material=@n11, WorldPositionOffset=@n1) {materialFunction=/Engine/Functions/MaterialLayerFunctions/MatLayerBlend_AddWorldPositionOffset}.Blended Material
    False=@n10
   ) {parameterName=Use Crystal Growth})
  n12 = Power(Base=ScalarParameter() {bUseCustomPrimitiveData=true, parameterName=EXP})
  n13 = NamedRerouteUsage()
  Max(A=@n12, B=@n13)
  n14 = MaterialFunctionCall() {materialFunction=/Engine/Functions/Engine_MaterialFunctions02/PivotPainter/PivotPainter_PerObjectData}
  NamedRerouteDeclaration(None=@n14.Pivot Position) {name=Pivot Position, nodeColor=<rgba 0.4297 1 0 1>}
  NamedRerouteDeclaration(None=@n14.X-Axis Vector) {name=X-Axis Vector, nodeColor=<rgba 0 0.3437 1 1>}
  n15 = NamedRerouteDeclaration(None=Transform(None=Constant3Vector() {constant=<rgba 0 1 0 0>}) {transformSourceType=TRANSFORMSOURCE_Local}) {name=Rotation Axis, nodeColor=<rgba 0.2891 1 0 1>}
  n16 = Add(A=Time(), B=NamedRerouteUsage())
  n17 = Sine(None=Multiply(A=Frac(None=@n16), B=ScalarParameter() {defaultValue=12.6, parameterName=Shake Frequency}) {constB=30.35})
  n18 = Floor(None=Add(A=ScalarParameter() {bUseCustomPrimitiveData=true, parameterName=EXP}) {constB=0.01})
  n19 = NamedRerouteDeclaration(None=Multiply(A=ScalarParameter() {defaultValue=0.004, parameterName=Crystals Max Angle, sliderMax=1}, B=Multiply(A=@n17, B=@n18))) {name=Rotation Angle, nodeColor=<rgba 0 0.2031 1 1>}
  n20 = NamedRerouteUsage()
  n21 = WorldPosition()
  NamedRerouteDeclaration(None=Multiply(A=RotateAboutAxis(NormalizedRotationAxis=@n15, RotationAngle=@n19, PivotPoint=@n20, Position=@n21.XYZ), B=NamedRerouteUsage())) {name=Wiggle WPO, nodeColor=<rgba 0.1484 1 0 1>}
  n22 = ScalarParameter() {bUseCustomPrimitiveData=true, parameterName=Break Crystals, primitiveDataIndex=4}
  n23 = NamedRerouteUsage()
  n24 = NamedRerouteUsage()
  NamedRerouteDeclaration(None=If(A=Multiply(A=@n22, B=@n23), "A > B"=NamedRerouteUsage(), "A == B"=@n24, "A < B"=@n24)) {name=Scale WPO, nodeColor=<rgba 0 1 0.625 1>}
  n25 = Time() {period=2}
  n26 = Multiply(A=Abs(None=NamedRerouteUsage()), B=ScalarParameter() {defaultValue=0.25, parameterName=Animation Randomness})
  n27 = Divide(
   A=Subtract(A=Subtract(A=@n25, B=@n26), B=ScalarParameter() {bUseCustomPrimitiveData=true, parameterName=Anim Time, primitiveDataIndex=2})
   B=ScalarParameter() {defaultValue=2.2, parameterName=Anim Duration}
  )
  n28 = Multiply(
   A=NamedRerouteDeclaration(None=Saturate(None=@n27)) {name=Growth Animation Timeline, nodeColor=<rgba 0 1 0.7422 1>}
   B=ScalarParameter() {defaultValue=90, parameterName=CrystalGrowthDistance}
  ) {constA=1}
  n29 = Distance(
   A=TransformPosition(None=VectorParameter() {parameterName=CrystalGrowthOrigin, defaultValue=<rgba 0 62 10 0>}.RGB) {transformType=TRANSFORMPOSSOURCE_World}
   B=NamedRerouteUsage()
  )
  n30 = InverseLinearInterpolate(A=Subtract(A=@n28, B=ScalarParameter() {defaultValue=10, parameterName=Animation Blur distance}) {constB=10}, B=@n28, Value=@n29)
  n31 = NamedRerouteUsage()
  n32 = TransformPosition(None=VectorParameter() {parameterName=CrystalGrowthOrigin, defaultValue=<rgba 0 62 10 0>}.RGB) {transformType=TRANSFORMPOSSOURCE_World}
  n33 = MaterialFunctionCall!(
   1=ScalarParameter() {defaultValue=9, parameterName=CrystalGrowthDistanceMin}
   2=ScalarParameter() {defaultValue=32, parameterName=CrystalGrowthDistanceMid}
   3=ScalarParameter() {defaultValue=140, parameterName=CrystalGrowthDistanceMax}
   4
   A=Divide(A=ScalarParameter() {bUseCustomPrimitiveData=true, parameterName=Growth Level, primitiveDataIndex=1})
  ) {materialFunction=/Engine/Functions/Engine_MaterialFunctions02/Lerp_Multiple_Float}
  n34 = Max(A=Power(Base=Saturate(None=@n30)) {constExponent=20}, B=OneMinus(None=SphereMask(A=@n31, B=@n32, Radius=@n33.Lerp 3 Inputs)))
  NamedRerouteDeclaration(None=OneMinus(None=@n34)) {name=Crystals Growth Mask, nodeColor=<rgba 0 1 0.3672 1>}
  NamedRerouteDeclaration(None=@n2.BaseColor) {name=Base Color, nodeColor=<rgba 0 1 0.9297 1>}
  n35 = MaterialFunctionCall() {materialFunction=/Game/Packages/Fortress_SharedMaterials/Base_Material_Functions/BaseCustomizationFunctions/MF_Base_LocalPosition}
  n36 = PreSkinnedPosition()
  NamedRerouteDeclaration(None=FunctionInput(Preview=ShaderStageSwitch(PixelShader=@n35.Pre-Skinned Local Position Pixel Shader, VertexShader=@n36)) {bUsePreviewValueAsDefault=true, inputName=Preskinned Position, sortPriority=50}) {name=Preskinned Position, nodeColor=<rgba 0.121 0.2774 1 1>}
  n37 = PreSkinnedNormal()
  NamedRerouteDeclaration(None=FunctionInput(Preview=ShaderStageSwitch(PixelShader=@n35.Pre-Skinned Local Normal Pixel Shader, VertexShader=@n37)) {bUsePreviewValueAsDefault=true, inputName=Preskinned Normals, sortPriority=60}) {name=Preskinned Normals, nodeColor=<rgba 0.121 0.2774 1 1>}
  n38 = PreSkinnedLocalBounds()
  NamedRerouteDeclaration(None=FunctionInput(Preview=ShaderStageSwitch(PixelShader=@n35.Pre-Skinned Local Bounds Pixel Shader, VertexShader=@n38.Half Extents)) {bUsePreviewValueAsDefault=true, inputName=Preskinned Bounds, sortPriority=70}) {name=Preskinned Bounds, nodeColor=<rgba 0.121 0.2774 1 1>}
  NamedRerouteDeclaration(None=MaterialFunctionCall(Seed=@n14.Random Value Per Element, Float=@n14.Random Value Per Element) {materialFunction=&MF_Hash}.Float) {name=Random Value Per Element, nodeColor=<rgba 1 0.7031 0 1>}
  n39 = Time() {period=2}
  n40 = ScalarParameter() {bUseCustomPrimitiveData=true, parameterName=Break Crystals, primitiveDataIndex=4}
  n41 = Saturate(None=Divide(A=Subtract(A=@n39, B=@n40), B=ScalarParameter() {defaultValue=0.13, parameterName=Break Duration}))
  n42 = Multiply(A=@n41, B=Subtract(A=NamedRerouteUsage(), B=WorldPosition().XYZ) {constA=0})
  n43 = ScalarParameter() {defaultValue=30, parameterName=Break Distance}
  n44 = LinearInterpolate(B=NamedRerouteUsage(), Alpha=ScalarParameter() {defaultValue=0.5, parameterName=Randomize Crystal Distance}) {constA=1}
  n45 = Multiply(A=Multiply(A=@n41, B=Multiply(A=@n43, B=@n44)), B=NamedRerouteUsage())
  NamedRerouteDeclaration(None=Multiply(A=Add(A=@n42, B=@n45), B=NamedRerouteUsage())) {name=Break Animation, nodeColor=<rgba 0.8516 0 1 1>}
  n46 = VectorParameter() {parameterName=XP Glow Channel, defaultValue=<rgba 0 0 1 0>}
  n47 = AppendVector(A=@n46.RGB, B=@n46.RGB)
  n48 = Multiply(A=Rotator(Coordinate=TextureCoordinate()), B=ScalarParameter() {defaultValue=1.5, parameterName=XP Glow Tiling})
  n49 = ScalarParameter() {defaultValue=-4, parameterName=XP Glow BumpOffset Height}
  n50 = Time()
  n51 = Multiply(A=NamedRerouteUsage()) {constB=65.314}
  n52 = ScalarParameter() {defaultValue=-1.5, parameterName=XP Glow Panning Speed}
  n53 = TextureSampleParameter2D(UVs=Panner(
    Coordinate=BumpOffset(Coordinate=@n48, Height=@n49)
    Time=Add(A=@n50, B=@n51)
    Speed=MaterialFunctionCall(X=@n52) {materialFunction=&MakeFloat2}.Result
   ) {speedX=-0.1}) {parameterName=XP Glow Texture, texture=/Game/TechArt/Seasonal/LackLuster/Textures/T_Lackluster_Petrified_Crystal_Pack}
  n54 = Multiply(A=ScalarParameter() {defaultValue=90, parameterName=XP Glow Intensity}, B=Multiply(A=DotProduct(A=@n47, B=@n53.RGBA), B=Add(A=@n12, B=@n13))) {constB=4}
  n55 = WorldPosition()
  n56 = NamedRerouteUsage()
  n57 = Saturate(None=Power(Base=Distance(A=@n55.XYZ, B=@n56), Exp=ScalarParameter() {defaultValue=3, parameterName=Glow Exponent}) {constExponent=1})
  n58 = Fresnel() {baseReflectFraction=0, exponent=0.2}
  n59 = Multiply(A=Multiply(A=Add(A=@n54, B=@n57), B=OneMinus(None=@n58)), B=NamedRerouteUsage())
  n60 = ScalarParameter() {bUseCustomPrimitiveData=true, parameterName=EXP}
  n61 = Add(
   A=Multiply(A=Power(Base=@n60), B=ScalarParameter() {defaultValue=3, parameterName=Crystal Emissive Multiplier})
   B=ScalarParameter() {defaultValue=0.3, parameterName=Crystal Emissive Bias}
  )
  NamedRerouteDeclaration(None=Add(A=NamedRerouteUsage(), B=Multiply(A=@n59, B=@n61))) {name=Emissive Out, nodeColor=<rgba 0.125 0 1 1>}
  n62 = InverseLinearInterpolate(A=Subtract(A=@n28, B=ScalarParameter() {defaultValue=19, parameterName=Glow Blur Distance}) {constB=10}, B=@n28, Value=@n29)
  NamedRerouteDeclaration(None=Multiply(A=Saturate(None=@n62), B=ScalarParameter() {defaultValue=30, parameterName=Crystal Initial Glow Intensity})) {name=Crystals Glow Mask, nodeColor=<rgba 0 0.8828 1 1>}
  n63 = NamedRerouteUsage()
  n64 = WorldPosition()
  n65 = ScalarParameter() {bUseCustomPrimitiveData=true, parameterName=Growth Level, primitiveDataIndex=1}
  n66 = Multiply(
   A=MaterialFunctionCall(
    1=ScalarParameter() {defaultValue=0.7, parameterName=Crystals Size 1}
    2=ScalarParameter() {defaultValue=0.825, parameterName=Crystals Size 2}
    3=ScalarParameter() {defaultValue=1, parameterName=Crystals Size 3}
    A=Divide(A=@n65)
   ) {materialFunction=&Lerp_Multiple_Float}.Lerp 3 Inputs
   B=OneMinus(None=@n34)
  )
  NamedRerouteDeclaration(None=Multiply(A=Subtract(A=@n63, B=@n64.XYZ) {constA=0}, B=OneMinus(None=@n66))) {name=Grow Animation, nodeColor=<rgba 0 1 0.6016 1>}
 ]}
