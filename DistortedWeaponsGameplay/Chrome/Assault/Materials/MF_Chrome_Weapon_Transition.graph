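# Material function that blends an animated, spreading LackLuster chrome layer onto weapon materials, gated by the Use Chrome Transition static switch.
# A sphere mask from the crystal growth origin, with separate front and back radii and a sight-channel cutout, defines the chrome area.
# Custom primitive data feeds growth level and animation timing, with noise added to the transition.
# Mask derivatives form edge normals, Fresnel-flattened panning swirls add detail, and world position offset is overridden.
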MaterialFunction /DistortedWeaponsGameplay/Chrome/Assault/Materials/MF_Chrome_Weapon_Transition
graph {nodes [
  n1 = MaterialFunctionCall!(
   bBlendMaterialAttributes
   MA_In
   UVsOverride
   Override Normals
   BlendMaterialAttributesAlpha
   BaseColorOverride
   WorldPositionOverride=NamedRerouteUsage()
   VertexNormalOverride
   SourceNormalsStrengthOverride
  ) {materialFunction=/Game/TechArt/Seasonal/LackLuster/MaterialFunctions/MF_LackLuster_Base}
  n2 = NamedRerouteUsage()
  n3 = NamedRerouteUsage()
  n4 = FunctionInput() {bUsePreviewValueAsDefault=true, inputName=Mask, inputType=FunctionInput_Scalar, sortPriority=1000, previewValue=<1 0 0 1>}
  n5 = FunctionInput() {inputType=FunctionInput_MaterialAttributes}
  n6 = MaterialFunctionCall!(
   Material=BlendMaterialAttributes(
    A=@n5
    B=MaterialFunctionCall(Material=@n1.MA Out, Additional Normal=@n2) {materialFunction=/Engine/Functions/MaterialLayerFunctions/MatLayerBlend_BlendAngleCorrectedNormals}.Blended Material
    Alpha=Multiply(A=@n3, B=@n4)
   )
   Additional Normal=NamedRerouteUsage()
   Mask
  ) {materialFunction=&MatLayerBlend_BlendAngleCorrectedNormals}
  n7 = NamedRerouteUsage()
  FunctionOutput(None=StaticSwitchParameter(
    True=MaterialFunctionCall!(Material=@n6.Blended Material, Mask, WorldPositionOffset=@n7) {materialFunction=/Engine/Functions/MaterialLayerFunctions/MatLayerBlend_OverrideWorldPositionOffset}.Blended Material
    False=@n5
   ) {parameterName=Use Chrome Transition})
  n8 = MaterialFunctionCall() {materialFunction=/Game/Packages/Fortress_SharedMaterials/Base_Material_Functions/BaseCustomizationFunctions/MF_Base_LocalPosition}
  n9 = PreSkinnedPosition()
  NamedRerouteDeclaration(None=FunctionInput(Preview=ShaderStageSwitch(PixelShader=@n8.Pre-Skinned Local Position Pixel Shader, VertexShader=@n9)) {bUsePreviewValueAsDefault=true, inputName=Preskinned Position, sortPriority=50}) {name=Preskinned Position, nodeColor=<rgba 0.121 0.2774 1 1>}
  n10 = PreSkinnedNormal()
  NamedRerouteDeclaration(None=FunctionInput(Preview=ShaderStageSwitch(PixelShader=@n8.Pre-Skinned Local Normal Pixel Shader, VertexShader=@n10)) {bUsePreviewValueAsDefault=true, inputName=Preskinned Normals, sortPriority=60}) {name=Preskinned Normals, nodeColor=<rgba 0.121 0.2774 1 1>}
  n11 = ChannelMaskParameter(None=TextureSample() {texture=/DistortedWeaponsGameplay/Chrome/Assault/Textures/T_Sight_Mask}.RGB) {parameterName=Sight Channel}
  n12 = TransformPosition(None=VectorParameter() {parameterName=CrystalGrowthOrigin, defaultValue=<rgba 0 10 10 0>}.RGB) {transformType=TRANSFORMPOSSOURCE_World}
  n13 = WorldPosition()
  n14 = PreSkinnedPosition()
  n15 = VertexInterpolator(VS=@n14)
  n16 = Divide(A=ComponentMask(None=ShaderStageSwitch(PixelShader=@n15.PS, VertexShader=@n14)) {g=true}) {constB=10}
  n17 = OneMinus(None=Saturate(None=ConstantBiasScale(None=@n16)))
  n18 = LinearInterpolate(
   A=LinearInterpolate(
    A=ScalarParameter() {defaultValue=10, parameterName=Front Min Chrome Spread}
    B=ScalarParameter() {defaultValue=10, parameterName=Back Min Chrome Spread}
    Alpha=@n17
   )
   B=LinearInterpolate(
    A=ScalarParameter() {defaultValue=140, parameterName=Front Max Chrome Spread}
    B=ScalarParameter() {defaultValue=50, parameterName=Back Max Chrome Spread}
    Alpha=@n17
   )
   Alpha=Power(Base=NamedRerouteUsage())
  ) {constA=10, constAlpha=1, constB=65}
  n19 = ScalarParameter() {defaultValue=0.98, parameterName=Chrome Feathering Hardness}
  n20 = NamedRerouteDeclaration(None=Multiply(A=OneMinus(None=@n11.R), B=SphereMask(A=@n12, B=@n13.XYZ, Radius=@n18, Hardness=@n19) {hardnessPercent=90})) {name=Chrome Mask, nodeColor=<rgba 0.0078 1 0 1>}
  n21 = DDX(Value=@n20)
  n22 = DDY(Value=@n20)
  n23 = ScalarParameter() {defaultValue=0.1, parameterName=Edge Z Normal}
  NamedRerouteDeclaration(None=Normalize(VectorInput=MaterialFunctionCall(X=@n21, Y=@n22, Z=@n23) {materialFunction=&MakeFloat3}.Result)) {name=Edge Normal, nodeColor=<rgba 0.7344 1 0 1>}
  NamedRerouteDeclaration(None=TextureCoordinate()) {name=UVs, nodeColor=<rgba 0.2891 0 1 1>}
  n24 = GetMaterialAttributes(None=@n5) {attributeGetTypes [#1, #2]}
  NamedRerouteDeclaration(None=@n24.WorldPositionOffset) {name=Original WPO, nodeColor=<rgba 0.2891 1 0 1>}
  n25 = TextureSampleParameter2D(UVs=NamedRerouteUsage()) {
   parameterName=ChromeNormals
   samplerSource=SSM_Wrap_WorldGroupSettings
   samplerType=SAMPLERTYPE_Normal
   texture=/Game/TechArt/Seasonal/LackLuster/Textures/T_Lackluster_Swirl_N
  }
  n26 = TextureSampleParameter2D(UVs=NamedRerouteUsage()) {
   parameterName=ChromeNormals
   samplerSource=SSM_Wrap_WorldGroupSettings
   samplerType=SAMPLERTYPE_Normal
   texture=&T_Lackluster_Swirl_N
  }
  n27 = ShadingPathSwitch(Default=Add(A=@n25.RGB, B=@n26.RGB), Mobile=@n25.RGB)
  n28 = ScalarParameter() {defaultValue=0.9, parameterName=Flatten Swirl Facing}
  n29 = ScalarParameter() {defaultValue=0.3, parameterName=Flatten Swirl Side}
  n30 = Fresnel(Normal=VertexNormalWS()) {baseReflectFraction=0, exponent=3}
  NamedRerouteDeclaration(None=MaterialFunctionCall(Normal=Normalize(VectorInput=@n27), Flatness=LinearInterpolate(A=@n28, B=@n29, Alpha=@n30)) {materialFunction=&FlattenNormal}.Result) {name=Swirl Normal, nodeColor=<rgba 1 0.0469 0 1>}
  n31 = Divide(
   A=Subtract(A=Time() {period=2}, B=ScalarParameter() {bUseCustomPrimitiveData=true, parameterName=Anim Time, primitiveDataIndex=2})
   B=ScalarParameter() {defaultValue=0.5, parameterName=Anim Duration}
  )
  NamedRerouteDeclaration(None=Max(A=ScalarParameter() {bUseCustomPrimitiveData=true, parameterName=Prevent Chrome Animation, primitiveDataIndex=3}, B=Saturate(None=@n31)) {constA=1}) {name=Growth Animation Timeline, nodeColor=<rgba 0 1 0.7422 1>}
  n32 = PreSkinnedLocalBounds()
  NamedRerouteDeclaration(None=FunctionInput(Preview=ShaderStageSwitch(PixelShader=@n8.Pre-Skinned Local Bounds Pixel Shader, VertexShader=@n32.Half Extents)) {bUsePreviewValueAsDefault=true, inputName=Preskinned Bounds, sortPriority=70}) {name=Preskinned Bounds, nodeColor=<rgba 0.121 0.2774 1 1>}
  n33 = ScalarParameter() {parameterName=Level Preview}
  n34 = ScalarParameter() {bUseCustomPrimitiveData=true, parameterName=Growth Level, primitiveDataIndex=1}
  n35 = NamedRerouteUsage()
  n36 = Divide(A=Add(A=Max(A=@n33, B=@n34), B=Subtract(A=@n35)), B=ScalarParameter() {defaultValue=3, parameterName=Max Level}) {constB=5}
  n37 = TextureObject() {texture=/Game/Effects/Fort_Effects/Textures/Tile/T_GaussianNoise_01_M}
  n38 = Constant3Vector() {constant=<rgba 15 15 15 0>}
  n39 = NamedRerouteUsage()
  n40 = NamedRerouteUsage()
  n41 = ComponentMask(None=NamedRerouteDeclaration(None=MaterialFunctionCall(TextureObject=@n37, TextureSize=@n38, WorldPosition=@n39, World Space Normal=@n40) {materialFunction=/Game/Weapons/WeaponSkins/Wraps/MaterialFunctions/WorldAlignedTexture}.XY Texture) {name=Noise, nodeColor=<rgba 1 0 0.8203 1>}) {r=true}
  n42 = ScalarParameter() {defaultValue=0.3, parameterName=Growth Noise}
  NamedRerouteDeclaration(None=Add(A=Saturate(None=@n36), B=Multiply(A=@n41, B=@n42) {constB=0.1})) {name=Transition Mask, nodeColor=<rgba 1 0.375 0 1>}
  NamedRerouteDeclaration(None=SphereMask(A=@n12, B=@n13.XYZ, Radius=@n18, Hardness=ScalarParameter() {defaultValue=0.7, parameterName=First Hardness})) {name=Chrome Edge Normal, nodeColor=<rgba 0.9219 1 0 1>}
  n43 = NamedRerouteUsage()
  n44 = Multiply(A=ScalarParameter() {defaultValue=3, parameterName=ChromePanSpeedScale}, B=Time())
  NamedRerouteDeclaration(None=Panner(Coordinate=NamedRerouteDeclaration(None=@n43) {name=Chrome_UVs_A_NoPanning, nodeColor=<rgba 0 1 0.7656 1>}, Time=@n44) {bFractionalPart=true, speedX=0.0008, speedY=0.0022}) {name=Chrome_UVs_A, nodeColor=<rgba 0.9219 0 1 1>}
  NamedRerouteDeclaration(None=Panner(Coordinate=NamedRerouteUsage(), Time=@n44) {bFractionalPart=true, speedX=0.0054, speedY=0.0022}) {name=Chrome_UVs_B, nodeColor=<rgba 0 0.3906 1 1>}
  NamedRerouteDeclaration(None=@n24.Normal) {name=Normal In, nodeColor=<rgba 0 0.4141 1 1>}
 ]}
