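# Master material for Chrome weapons, layering crystal and transition effects over the shared Fortnite weapon master.
# Material attributes chain through an overheat function (skipped on mobile), Chrome crystals, Chrome transition, then ADS scope master.
# A channel-selected crystal gradient drives tip glow and optionally masks the transition.
# Final emissive is taken straight from the crystal stage, bypassing later functions, and a switch can replace world position offset with vertex-color element hiding.
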Material /DistortedWeaponsGameplay/Chrome/Assault/Materials/M_Chrome_Weapon_MASTER
graph {
 properties {
  bAutomaticallySetUsageInEditor=false
  bUseMaterialAttributes=true
  bUsedWithClothing=true
  bUsedWithMeshDeformer=true
  bUsedWithMorphTargets=true
  bUsedWithNanite=true
  bUsedWithSkeletalMesh=true
  pixelDepthOffsetMode=PDOM_Legacy
  usedShadingModels=Default Lit
 }
 nodes [
  n1 = MaterialFunctionCall!(In) {materialFunction=/Game/Packages/Fortress_SharedMaterials/Base_Material_Functions/BaseCustomizationFunctions/MF_FN_Weapon_MASTER}
  n2 = ScalarParameter() {defaultValue=0.3, parameterName=Crystals Glow SmoothS Min}
  n3 = ScalarParameter() {defaultValue=1, parameterName=Crystals Glow SmoothS Max}
  n4 = TextureSampleParameter2D() {parameterName=Crystals Mask, samplerType=SAMPLERTYPE_Masks, texture=/DistortedWeaponsGameplay/Chrome/Assault/Textures/T_Crystals_Gradient}
  n5 = ChannelMaskParameter(None=@n4.RGB) {parameterName=Crystal Mask Channel}
  n6 = MaterialFunctionCall!(
   In=ShadingPathSwitch(
    Default=MaterialFunctionCall!(Base MA=@n1.Material Result, Pre-Skinned Local Position=@n1.Pre-Skinned Local Position Pixel Shader) {materialFunction=/Game/Weapons/WeaponSkins/MaterialFunction/OverheatGunFunction}.Result
    Mobile=@n1.Material Result
   ) {desc=Migrated from FeatureLevelSwitch}
   Preskinned Position
   Preskinned Normals
   Preskinned Bounds
   Upgrade Emissive Multiplier=LinearInterpolate(
    A=ScalarParameter() {defaultValue=0.2, parameterName=Crystals Tip Glow}
    Alpha=SmoothStep(Min=@n2, Max=@n3, Value=@n5.R) {constMax=0.85, constMin=0.5}
   )
  ) {materialFunction=&/MF_Chrome_Weapon_Crystals}
  n7 = StaticSwitchParameter(True=SmoothStep(Value=@n5.R) {constMax=0.85, constMin=0.5}, False=Constant() {r=1}) {parameterName=Use Crystal Mask}
  n8 = GetMaterialAttributes(None=MaterialFunctionCall!(Material Attributes=MaterialFunctionCall!(In=@n6.Result, Preskinned Position, Preskinned Normals, Preskinned Bounds, Mask=@n7) {materialFunction=&/MF_Chrome_Weapon_Transition}.Result) {materialFunction=/Game/Packages/Fortnite_UniqueMaterials_1/Weapons/ADSScopeMaster}.Result) {attributeGetTypes [#1]}
  SetMaterialAttributes!(
   MaterialAttributes=@n8.MaterialAttributes
   World Position Offset=StaticSwitchParameter(
    True=MaterialFunctionCall!(WPO) {materialFunction=/Game/Packages/Fortress_SharedMaterials/Base_Material_Functions/Fortnite_HideElementsWithVertexColor}.Result
    False=@n8.World Position Offset
   ) {group=Opacity Mask, parameterName=Use Vertex Colors for Mask}
   Emissive Color=GetMaterialAttributes(None=@n6.Result) {attributeGetTypes [#2]}.Emissive Color
  ) {attributeSetTypes [#1, #2]}
 ]
}
