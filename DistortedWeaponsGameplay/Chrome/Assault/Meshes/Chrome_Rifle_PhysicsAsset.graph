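# Physics asset for the Chrome assault rifle skeletal mesh in the DistortedWeaponsGameplay content.
# It holds a single body on the handle bone, shaped as one capsule laid on its side and offset along the bone's X axis to approximate the weapon.
# There are no constraints, so the rifle simulates and collides as one rigid piece.
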PhysicsAsset /DistortedWeaponsGameplay/Chrome/Assault/Meshes/Chrome_Rifle_PhysicsAsset
bodies {handle [{shapeType=Capsule, length=47.51, radius=15.7653, center=<24.2483 0.0395 3.0121>, rotation=<rot 0 89.9657 -86.2165>}]}
constraints []
