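# Sound cue for the arm-reach foley sound of the Chrome Assault weapon.
# A single wave player for Chrome_Assault_ArmReach_01 feeds a Modulator that slightly randomizes pitch and volume on each play.
# The cue boosts overall volume, uses the Fort_Player_Foley sound class and takes its attenuation from the shared basic foley attenuation asset.
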SoundCue /DistortedWeaponsGameplay/Chrome/Assault/Sounds/Chrome_Assault_ArmReach_Cue
root=Modulator_2
nodes [
 {id=Modulator_2, children [WavePlayer_1], properties {pitchMax=1.1, pitchMin=1, volumeMax=1, volumeMin=0.8}}
 {id=WavePlayer_1, properties {soundWaveAssetPtr=&/Chrome_Assault_ArmReach_01}}
]
properties {
 attenuationSettings=&Foley_Attenuation_Basic_Att
 maxDistance=1536
 soundClassObject=&Fort_Player_Foley
 volumeMultiplier=1.5
 attenuationOverrides {bAttenuate=false}
 concurrencyOverrides {maxCount=1, resolutionRule=PreventNew}
}
