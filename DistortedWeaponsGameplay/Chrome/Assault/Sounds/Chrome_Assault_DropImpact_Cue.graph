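# Sound cue for the Chrome Assault weapon's drop impact in Distorted Weapons gameplay, classed under Fort_Loot_Collect.
# A single wave player feeds a Modulator root that randomizes pitch and volume on each play, with the cue's overall volume boosted.
# Attenuation is overridden on the cue for short-range falloff.
# The shared Weapon_Drop_StopOldest concurrency set limits overlapping drop sounds, as the cue's own concurrency override values are not switched on.
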SoundCue /DistortedWeaponsGameplay/Chrome/Assault/Sounds/Chrome_Assault_DropImpact_Cue
root=Modulator_0
nodes [
 {id=Modulator_0, children [WavePlayer_2], properties {pitchMax=1.1, volumeMax=1.1, volumeMin=0.9}}
 {id=WavePlayer_2, properties {soundWaveAssetPtr=&/Chrome_Assault_DropImpact_01}}
]
properties {
 bOverrideAttenuation=true
 maxDistance=1000
 soundClassObject=&Fort_Loot_Collect
 volumeMultiplier=1.6
 attenuationOverrides {falloffDistance=1000, attenuationShapeExtents {x=0}}
 concurrencyOverrides {maxCount=3, resolutionRule=PreventNew}
 concurrencySet [&Weapon_Drop_StopOldest]
}
