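# Sound cue for the drop sound of the Chrome assault weapon.
# It plays a single drop wave through a modulator that slightly randomizes pitch and volume on each play.
# It uses the shared weapon drop attenuation and the loot collect sound class, with a small volume boost.
# It belongs to the Weapon_Drop_StopOldest concurrency set to limit overlapping drop sounds.
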SoundCue /DistortedWeaponsGameplay/Chrome/Assault/Sounds/Chrome_Assault_Drop_Cue
root=Modulator_0
nodes [
 {id=Modulator_0, children [WavePlayer_1], properties {pitchMax=1.1, pitchMin=0.9, volumeMax=1.1, volumeMin=0.9}}
 {id=WavePlayer_1, properties {soundWaveAssetPtr=&/Chrome_Assault_Drop_01}}
]
properties {
 attenuationSettings=&Fort_Weapon_Drop_Att
 maxDistance=3000
 soundClassObject=&Fort_Loot_Collect
 volumeMultiplier=1.2
 concurrencyOverrides {resolutionRule=PreventNew}
 concurrencySet [&Weapon_Drop_StopOldest]
}
