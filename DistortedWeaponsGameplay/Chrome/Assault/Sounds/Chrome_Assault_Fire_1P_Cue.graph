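# First-person fire sound cue for the Chrome Assault weapon, layering a random close gunshot with a quieter, delayed shell-impact sound.
# Each layer passes through a modulator for variation before mixing, with the shell layer delayed about half a second.
# A quality switch gives the first level the full wave pools, while the other three share a smaller mix.
# First-person attenuation and owner-limited, stop-oldest concurrency govern playback.
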SoundCue /DistortedWeaponsGameplay/Chrome/Assault/Sounds/Chrome_Assault_Fire_1P_Cue
root=QualityLevel_0
nodes [
 {id=Delay_0, children [Modulator_1], properties {delayMax=0.6, delayMin=0.5}}
 {id=Delay_1, children [Modulator_3], properties {delayMax=0.6, delayMin=0.5}}
 {id=Mixer_0, children [Modulator_0,Delay_0], properties {inputVolume [1,0.12]}}
 {id=Mixer_1, children [Modulator_2,Delay_1], properties {inputVolume [1,0.12]}}
 {id=Modulator_0, children [Random_0]}
 {id=Modulator_1, children [Random_1], properties {pitchMax=1.1, pitchMin=0.9, volumeMax=1.1, volumeMin=0.9}}
 {id=Modulator_2, children [Random_2]}
 {id=Modulator_3, children [Random_3], properties {pitchMax=1.1, pitchMin=0.9, volumeMax=1.1, volumeMin=0.9}}
 {id=QualityLevel_0, children [Mixer_0,Mixer_1,Mixer_1,Mixer_1]}
 {id=Random_0, children [WavePlayer_1,WavePlayer_2,WavePlayer_3,WavePlayer_4,WavePlayer_0], properties {weights [1,1,1,1,1]}}
 {
  id=Random_1
  children [WavePlayer_5, WavePlayer_6, WavePlayer_7, WavePlayer_8, WavePlayer_9, WavePlayer_10, WavePlayer_11]
  properties {weights [1,1,1,1,1,1,1]}
 }
 {id=Random_2, children [WavePlayer_12, WavePlayer_16, WavePlayer_13], properties {weights [1,1,1]}}
 {id=Random_3, children [WavePlayer_17, WavePlayer_18, WavePlayer_19], properties {weights [1,1,1]}}
 {id=WavePlayer_0, properties {soundWaveAssetPtr=&/Chrome_Assault_Fire_Close_05}}
 {id=WavePlayer_1, properties {soundWaveAssetPtr=&/Chrome_Assault_Fire_Close_01}}
 {id=WavePlayer_10, properties {soundWaveAssetPtr=&/Chrome_Assault_ShellImpacts_06}}
 {id=WavePlayer_11, properties {soundWaveAssetPtr=&/Chrome_Assault_ShellImpacts_07}}
 {id=WavePlayer_12, properties {soundWaveAssetPtr=&Chrome_Assault_Fire_Close_01}}
 {id=WavePlayer_13, properties {soundWaveAssetPtr=&/Chrome_Assault_Fire_Close_03}}
 {id=WavePlayer_16, properties {soundWaveAssetPtr=&/Chrome_Assault_Fire_Close_02}}
 {id=WavePlayer_17, properties {soundWaveAssetPtr=&/Chrome_Assault_ShellImpacts_01}}
 {id=WavePlayer_18, properties {soundWaveAssetPtr=&/Chrome_Assault_ShellImpacts_02}}
 {id=WavePlayer_19, properties {soundWaveAssetPtr=&/Chrome_Assault_ShellImpacts_03}}
 {id=WavePlayer_2, properties {soundWaveAssetPtr=&Chrome_Assault_Fire_Close_02}}
 {id=WavePlayer_3, properties {soundWaveAssetPtr=&Chrome_Assault_Fire_Close_03}}
 {id=WavePlayer_4, properties {soundWaveAssetPtr=&/Chrome_Assault_Fire_Close_04}}
 {id=WavePlayer_5, properties {soundWaveAssetPtr=&Chrome_Assault_ShellImpacts_01}}
 {id=WavePlayer_6, properties {soundWaveAssetPtr=&Chrome_Assault_ShellImpacts_02}}
 {id=WavePlayer_7, properties {soundWaveAssetPtr=&Chrome_Assault_ShellImpacts_03}}
 {id=WavePlayer_8, properties {soundWaveAssetPtr=&/Chrome_Assault_ShellImpacts_04}}
 {id=WavePlayer_9, properties {soundWaveAssetPtr=&/Chrome_Assault_ShellImpacts_05}}
]
properties {
 attenuationSettings=&Fort_Weapon_Projectile_FirstPerson_Athena_Att
 bDebug=true
 maxDistance=2097152
 priority=5
 soundClassObject=&Fort_WeaponRanged_Fire_1P_PSM
 volumeMultiplier=0.87
 attenuationOverrides {manualReverbSendLevel=-1.0165, reverbSendMethod=Manual}
 concurrencyOverrides {bLimitToOwner=true, maxCount=2, resolutionRule=StopOldest}
 concurrencySet [&Weapon_Gunfire_FirstPerson_Conc]
}
