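# Sound cue for first-person firing of the Chrome Assault Tier 3 weapon, layering close gunshots with shell impacts.
# Each branch mixes a random, volume-boosted shot with a quieter, delayed random shell impact that varies in pitch and volume.
# A quality switch gives its first level larger wave pools, while the other three share a smaller-pool branch.
# First-person attenuation, manual reverb send and gunfire concurrency sets govern playback.
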SoundCue /DistortedWeaponsGameplay/Chrome/Assault/Sounds/Chrome_Assault_Fire_1P_Tier3_Cue
root=QualityLevel_0
nodes [
 {id=Delay_0, children [Modulator_3], properties {delayMax=0.6, delayMin=0.5}}
 {id=Delay_1, children [Modulator_2], properties {delayMax=0.6, delayMin=0.5}}
 {id=Mixer_0, children [Modulator_5,Delay_1], properties {inputVolume [1,0.15]}}
 {id=Mixer_2, children [Modulator_4,Delay_0], properties {inputVolume [1,0.15]}}
 {id=Modulator_0, children [Mixer_2]}
 {id=Modulator_1, children [Mixer_0]}
 {id=Modulator_2, children [Random_2], properties {pitchMax=1.1, pitchMin=0.9, volumeMax=1.1, volumeMin=0.9}}
 {id=Modulator_3, children [Random_0], properties {pitchMax=1.1, pitchMin=0.9, volumeMax=1.1, volumeMin=0.9}}
 {id=Modulator_4, children [Random_1], properties {pitchMax=1, pitchMin=1, volumeMax=1.4, volumeMin=1.3}}
 {id=Modulator_5, children [Random_3], properties {pitchMax=1, pitchMin=1, volumeMax=1.4, volumeMin=1.3}}
 {id=QualityLevel_0, children [Modulator_0,Modulator_1,Modulator_1,Modulator_1]}
 {
  id=Random_0
  children [WavePlayer_5,WavePlayer_1,WavePlayer_2,WavePlayer_3,WavePlayer_4,WavePlayer_0,WavePlayer_6]
  properties {weights [1,1,1,1,1,1,1]}
 }
 {id=Random_1, children [WavePlayer_7, WavePlayer_8, WavePlayer_9, WavePlayer_15, WavePlayer_16], properties {weights [1,1,1,1,1]}}
 {id=Random_2, children [WavePlayer_10, WavePlayer_11, WavePlayer_12], properties {weights [1,1,1]}}
 {id=Random_3, children [WavePlayer_19, WavePlayer_20, WavePlayer_21], properties {weights [1,1,1]}}
 {id=WavePlayer_0, properties {soundWaveAssetPtr=&/Chrome_Assault_ShellImpacts_06}}
 {id=WavePlayer_1, properties {soundWaveAssetPtr=&/Chrome_Assault_ShellImpacts_02}}
 {id=WavePlayer_10, properties {soundWaveAssetPtr=&/Chrome_Assault_ShellImpacts_01}}
 {id=WavePlayer_11, properties {soundWaveAssetPtr=&Chrome_Assault_ShellImpacts_02}}
 {id=WavePlayer_12, properties {soundWaveAssetPtr=&/Chrome_Assault_ShellImpacts_03}}
 {id=WavePlayer_15, properties {soundWaveAssetPtr=&/Chrome_Assault_Fire_Close_Tier03_Mix_04}}
 {id=WavePlayer_16, properties {soundWaveAssetPtr=&/Chrome_Assault_Fire_Close_Tier03_Mix_05}}
 {id=WavePlayer_19, properties {soundWaveAssetPtr=&/Chrome_Assault_Fire_Close_Tier03_Mix_01}}
 {id=WavePlayer_2, properties {soundWaveAssetPtr=&Chrome_Assault_ShellImpacts_03}}
 {id=WavePlayer_20, properties {soundWaveAssetPtr=&/Chrome_Assault_Fire_Close_Tier03_Mix_02}}
 {id=WavePlayer_21, properties {soundWaveAssetPtr=&/Chrome_Assault_Fire_Close_Tier03_Mix_03}}
 {id=WavePlayer_3, properties {soundWaveAssetPtr=&/Chrome_Assault_ShellImpacts_04}}
 {id=WavePlayer_4, properties {soundWaveAssetPtr=&/Chrome_Assault_ShellImpacts_05}}
 {id=WavePlayer_5, properties {soundWaveAssetPtr=&Chrome_Assault_ShellImpacts_01}}
 {id=WavePlayer_6, properties {soundWaveAssetPtr=&/Chrome_Assault_ShellImpacts_07}}
 {id=WavePlayer_7, properties {soundWaveAssetPtr=&Chrome_Assault_Fire_Close_Tier03_Mix_01}}
 {id=WavePlayer_8, properties {soundWaveAssetPtr=&Chrome_Assault_Fire_Close_Tier03_Mix_02}}
 {id=WavePlayer_9, properties {soundWaveAssetPtr=&Chrome_Assault_Fire_Close_Tier03_Mix_03}}
]
properties {
 attenuationSettings=&Fort_Weapon_Projectile_FirstPerson_Athena_Att
 bDebug=true
 maxDistance=2097152
 priority=5
 soundClassObject=&Fort_WeaponRanged_Fire_1P_PSM
 volumeMultiplier=0.52
 attenuationOverrides {manualReverbSendLevel=-1.0165, reverbSendMethod=Manual}
 concurrencyOverrides {bLimitToOwner=true, maxCount=1, resolutionRule=StopOldest, voiceStealReleaseTime=0.3, volumeScaleAttackTime=0.3}
 concurrencySet [&/Chrome_Weapon_Tier_Fire, &Weapon_Gunfire_Global_Conc]
}
