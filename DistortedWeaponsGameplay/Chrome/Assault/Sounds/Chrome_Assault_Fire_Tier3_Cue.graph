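# Sound cue for the Chrome assault weapon's tier 3 fire, layering close, medium and distant gunshot sounds.
# Each layer randomly picks a wave, then passes through modulation, distance-band attenuation and a third-person or distant fire sound class before nested mixers blend them.
# The quality switch gives the first level five variations per layer, while the other three share a three-variation branch.
# Weapon-tier and global gunfire concurrency sets govern voices.
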SoundCue /DistortedWeaponsGameplay/Chrome/Assault/Sounds/Chrome_Assault_Fire_Tier3_Cue
root=QualityLevel_0
nodes [
 {id=Attenuation_0, children [Modulator_9], properties {attenuationSettings=&Fort_Weapon_Projectile_Close_Athena_Att}}
 {
  id=Attenuation_1
  children [Modulator_5]
  properties {attenuationSettings=&Fort_Weapon_Projectile_Medium_Mono_Athena_Att}
 }
 {
  id=Attenuation_2
  children [Modulator_8]
  properties {attenuationSettings=&Fort_Weapon_Projectile_Distant_Mono_Athena_Att}
 }
 {id=Attenuation_3, children [Modulator_4], properties {attenuationSettings=&Fort_Weapon_Projectile_Close_Athena_Att}}
 {
  id=Attenuation_4
  children [Modulator_6]
  properties {attenuationSettings=&Fort_Weapon_Projectile_Medium_Mono_Athena_Att}
 }
 {
  id=Attenuation_5
  children [Modulator_7]
  properties {attenuationSettings=&Fort_Weapon_Projectile_Distant_Mono_Athena_Att}
 }
 {id=Mixer_0, children [Modulator_0,SoundClass_5], properties {inputVolume [1,0.35]}}
 {id=Mixer_1, children [Modulator_2,SoundClass_4], properties {inputVolume [1,0.6]}}
 {id=Mixer_2, children [Modulator_1,SoundClass_2], properties {inputVolume [1,0.35]}}
 {id=Mixer_3, children [Modulator_3,SoundClass_1], properties {inputVolume [1,0.6]}}
 {id=Modulator_0, children [Mixer_1], properties {pitchMax=1, pitchMin=1}}
 {id=Modulator_1, children [Mixer_3], properties {pitchMax=1, pitchMin=1}}
 {id=Modulator_2, children [SoundClass_3], properties {volumeMax=1, volumeMin=1}}
 {id=Modulator_3, children [SoundClass_0], properties {volumeMax=1, volumeMin=1}}
 {id=Modulator_4, children [Random_3], properties {pitchMax=1, pitchMin=1, volumeMax=1.1, volumeMin=1}}
 {id=Modulator_5, children [Random_1], properties {pitchMin=1, volumeMax=1.6, volumeMin=1.5}}
 {id=Modulator_6, children [Random_0], properties {pitchMin=1, volumeMax=1.6, volumeMin=1.5}}
 {id=Modulator_7, children [Random_2], properties {volumeMax=1.5, volumeMin=1.4}}
 {id=Modulator_8, children [Random_4], properties {volumeMax=1.5, volumeMin=1.4}}
 {id=Modulator_9, children [Random_5], properties {pitchMax=1, pitchMin=1, volumeMax=1.1, volumeMin=1}}
 {id=QualityLevel_0, children [Mixer_0,Mixer_2,Mixer_2,Mixer_2]}
 {id=Random_0, children [WavePlayer_0,WavePlayer_1,WavePlayer_2,WavePlayer_3,WavePlayer_4], properties {weights [1,1,1,1,1]}}
 {id=Random_1, children [WavePlayer_6,WavePlayer_7,WavePlayer_8], properties {weights [1,1,1]}}
 {id=Random_2, children [WavePlayer_5, WavePlayer_21, WavePlayer_22, WavePlayer_23, WavePlayer_24], properties {weights [1,1,1,1,1]}}
 {id=Random_3, children [WavePlayer_11, WavePlayer_12, WavePlayer_13, WavePlayer_15, WavePlayer_16], properties {weights [1,1,1,1,1]}}
 {id=Random_4, children [WavePlayer_14, WavePlayer_17, WavePlayer_18], properties {weights [1,1,1]}}
 {id=Random_5, children [WavePlayer_25, WavePlayer_26, WavePlayer_27], properties {weights [1,1,1]}}
 {
  id=SoundClass_0
  children [Attenuation_0]
  properties {soundClassOverride=&Fort_WeaponRanged_Fire_3P_PSM}
 }
 {
  id=SoundClass_1
  children [Attenuation_1]
  properties {soundClassOverride=&Fort_WeaponRanged_Fire_Distant}
 }
 {
  id=SoundClass_2
  children [Attenuation_2]
  properties {soundClassOverride=&Fort_WeaponRanged_Fire_Distant}
 }
 {
  id=SoundClass_3
  children [Attenuation_3]
  properties {soundClassOverride=&Fort_WeaponRanged_Fire_3P_PSM}
 }
 {
  id=SoundClass_4
  children [Attenuation_4]
  properties {soundClassOverride=&Fort_WeaponRanged_Fire_Distant}
 }
 {
  id=SoundClass_5
  children [Attenuation_5]
  properties {soundClassOverride=&Fort_WeaponRanged_Fire_Distant}
 }
 {id=WavePlayer_0, properties {soundWaveAssetPtr=&/Chrome_Assault_Fire_Medium_Tier03_Mix_01}}
 {id=WavePlayer_1, properties {soundWaveAssetPtr=&/Chrome_Assault_Fire_Medium_Tier03_Mix_02}}
 {id=WavePlayer_11, properties {soundWaveAssetPtr=&/Chrome_Assault_Fire_Close_Tier03_Mix_01}}
 {id=WavePlayer_12, properties {soundWaveAssetPtr=&/Chrome_Assault_Fire_Close_Tier03_Mix_02}}
 {id=WavePlayer_13, properties {soundWaveAssetPtr=&/Chrome_Assault_Fire_Close_Tier03_Mix_03}}
 {id=WavePlayer_14, properties {soundWaveAssetPtr=&/Chrome_Assault_Fire_Distant_Tier03_Mix_01}}
 {id=WavePlayer_15, properties {soundWaveAssetPtr=&/Chrome_Assault_Fire_Close_Tier03_Mix_04}}
 {id=WavePlayer_16, properties {soundWaveAssetPtr=&/Chrome_Assault_Fire_Close_Tier03_Mix_05}}
 {id=WavePlayer_17, properties {soundWaveAssetPtr=&/Chrome_Assault_Fire_Distant_Tier03_Mix_02}}
 {id=WavePlayer_18, properties {soundWaveAssetPtr=&/Chrome_Assault_Fire_Distant_Tier03_Mix_03}}
 {id=WavePlayer_2, properties {soundWaveAssetPtr=&/Chrome_Assault_Fire_Medium_Tier03_Mix_03}}
 {id=WavePlayer_21, properties {soundWaveAssetPtr=&Chrome_Assault_Fire_Distant_Tier03_Mix_02}}
 {id=WavePlayer_22, properties {soundWaveAssetPtr=&Chrome_Assault_Fire_Distant_Tier03_Mix_03}}
 {id=WavePlayer_23, properties {soundWaveAssetPtr=&/Chrome_Assault_Fire_Distant_Tier03_Mix_04}}
 {id=WavePlayer_24, properties {soundWaveAssetPtr=&/Chrome_Assault_Fire_Distant_Tier03_Mix_05}}
 {id=WavePlayer_25, properties {soundWaveAssetPtr=&Chrome_Assault_Fire_Close_Tier03_Mix_01}}
 {id=WavePlayer_26, properties {soundWaveAssetPtr=&Chrome_Assault_Fire_Close_Tier03_Mix_02}}
 {id=WavePlayer_27, properties {soundWaveAssetPtr=&Chrome_Assault_Fire_Close_Tier03_Mix_03}}
 {id=WavePlayer_3, properties {soundWaveAssetPtr=&/Chrome_Assault_Fire_Medium_Tier03_Mix_04}}
 {id=WavePlayer_4, properties {soundWaveAssetPtr=&/Chrome_Assault_Fire_Medium_Tier03_Mix_05}}
 {id=WavePlayer_5, properties {soundWaveAssetPtr=&Chrome_Assault_Fire_Distant_Tier03_Mix_01}}
 {id=WavePlayer_6, properties {soundWaveAssetPtr=&Chrome_Assault_Fire_Medium_Tier03_Mix_01}}
 {id=WavePlayer_7, properties {soundWaveAssetPtr=&Chrome_Assault_Fire_Medium_Tier03_Mix_02}}
 {id=WavePlayer_8, properties {soundWaveAssetPtr=&Chrome_Assault_Fire_Medium_Tier03_Mix_03}}
]
properties {
 bDebug=true
 maxDistance=27500
 volumeMultiplier=0.5
 concurrencyOverrides {bLimitToOwner=true, maxCount=2, resolutionRule=StopOldest, voiceStealReleaseTime=0.3, volumeScaleAttackTime=0.3}
 concurrencySet [&/Chrome_Weapon_Tier_Fire, &Weapon_Gunfire_Global_Conc]
}
