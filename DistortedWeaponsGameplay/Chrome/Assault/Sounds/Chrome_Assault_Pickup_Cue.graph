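# Sound cue that plays the pickup sound for the Chrome assault weapon.
# A single wave player for the pickup wave feeds a Modulator that slightly randomizes volume on each play.
# It uses the shared weapon pickup attenuation and the loot collect sound class, and joins a pickup concurrency set.
# Concurrency overrides are also stored, set to one voice that blocks new instances.
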SoundCue /DistortedWeaponsGameplay/Chrome/Assault/Sounds/Chrome_Assault_Pickup_Cue
root=Modulator_0
nodes [
 {id=Modulator_0, children [WavePlayer_2], properties {volumeMax=1.1, volumeMin=0.9}}
 {id=WavePlayer_2, properties {soundWaveAssetPtr=&/Chrome_Assault_Pickup_01}}
]
properties {
 attenuationSettings=&Fort_Weapon_Pickup_Att
 maxDistance=3500
 soundClassObject=&Fort_Loot_Collect
 volumeMultiplier=0.9231
 concurrencyOverrides {maxCount=1, resolutionRule=PreventNew}
 concurrencySet [&PickupMaximum_Oldest]
}
