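# Sound cue for the magazine insert sound during the Chrome Assault weapon's reload.
# It picks one of three insert wave variations at random with equal weight, then a modulator slightly varies the pitch.
# It plays as weapon foley with reload attenuation and shared local and global foley concurrency sets.
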SoundCue /DistortedWeaponsGameplay/Chrome/Assault/Sounds/Chrome_Assault_Reload_Insert_Cue
root=Modulator_13
nodes [
 {id=Modulator_13, children [Random_0], properties {pitchMax=1.1, pitchMin=0.9}}
 {id=Random_0, children [WavePlayer_0,WavePlayer_1,WavePlayer_2], properties {weights [1,1,1]}}
 {id=WavePlayer_0, properties {soundWaveAssetPtr=&/Chrome_Assault_Reload_Insert_01}}
 {id=WavePlayer_1, properties {soundWaveAssetPtr=&/Chrome_Assault_Reload_Insert_02}}
 {id=WavePlayer_2, properties {soundWaveAssetPtr=&/Chrome_Assault_Reload_Insert_03}}
]
properties {
 attenuationSettings=&Fort_Weapon_Reload_Att
 maxDistance=2250
 soundClassObject=&Fort_WeaponFoley
 volumeMultiplier=0.752
 attenuationOverrides {falloffDistance=1600}
 concurrencyOverrides {resolutionRule=PreventNew}
 concurrencySet [
  /Game/Sounds/Fort_Foley_Sounds/FoleyLibraries/Settings/Foley_Basic_Local_Conc
  /Game/Sounds/Fort_Foley_Sounds/FoleyLibraries/Settings/Foley_Basic_Global_Conc
 ]
}
