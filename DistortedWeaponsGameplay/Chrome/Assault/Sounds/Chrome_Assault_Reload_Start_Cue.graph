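# Sound cue for the start of the Chrome Assault weapon's reload, classed as weapon foley.
# A mixer layers a pitch-varied reload-start sound with a quieter, delayed magazine impact.
# The impact is picked at random from three variants and gets its own slight pitch modulation.
# It uses shared weapon reload attenuation and local and global basic foley concurrency sets.
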SoundCue /DistortedWeaponsGameplay/Chrome/Assault/Sounds/Chrome_Assault_Reload_Start_Cue
root=Mixer_0
nodes [
 {id=Delay_0, children [Modulator_1], properties {delayMax=0.5, delayMin=0.5}}
 {id=Mixer_0, children [Modulator_0,Delay_0], properties {inputVolume [1,0.6]}}
 {id=Modulator_0, children [WavePlayer_0], properties {pitchMax=1.1, pitchMin=0.9}}
 {id=Modulator_1, children [Random_1], properties {pitchMax=1.1, pitchMin=1}}
 {id=Random_1, children [WavePlayer_3,WavePlayer_4,WavePlayer_5], properties {weights [1,1,1]}}
 {id=WavePlayer_0, properties {soundWaveAssetPtr=&/Chrome_Assault_Reload_Start_01}}
 {id=WavePlayer_3, properties {soundWaveAssetPtr=&/Chrome_Assault_Magazine_Impact_01}}
 {id=WavePlayer_4, properties {soundWaveAssetPtr=&/Chrome_Assault_Magazine_Impact_02}}
 {id=WavePlayer_5, properties {soundWaveAssetPtr=&/Chrome_Assault_Magazine_Impact_03}}
]
properties {
 attenuationSettings=&Fort_Weapon_Reload_Att
 maxDistance=2250
 soundClassObject=&Fort_WeaponFoley
 volumeMultiplier=0.451
 attenuationOverrides {falloffDistance=1600}
 concurrencyOverrides {resolutionRule=PreventNew}
 concurrencySet [
  /Game/Sounds/Fort_Foley_Sounds/FoleyLibraries/Settings/Foley_Basic_Local_Conc
  /Game/Sounds/Fort_Foley_Sounds/FoleyLibraries/Settings/Foley_Basic_Global_Conc
 ]
}
