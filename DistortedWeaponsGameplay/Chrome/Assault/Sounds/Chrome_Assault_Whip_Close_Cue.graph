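# Sound cue for the close bullet whiz-by of the Chrome assault weapon, classed as a ranged projectile effect.
# A quality-level node sends its first two levels to a random pick of six whip waves and its last two to a pick of three.
# Each branch passes through a modulator that randomizes pitch and boosts volume variably.
# Attenuation is overridden off with zero stereo spread, and a shared Athena bullet whiz-by concurrency rule applies.
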SoundCue /DistortedWeaponsGameplay/Chrome/Assault/Sounds/Chrome_Assault_Whip_Close_Cue
root=QualityLevel_1
nodes [
 {id=Modulator_2, children [Random_2], properties {pitchMax=1.1, pitchMin=0.9, volumeMax=2.3, volumeMin=1.7}}
 {id=Modulator_3, children [Random_3], properties {pitchMax=1.1, pitchMin=0.9, volumeMax=2.3, volumeMin=1.7}}
 {id=QualityLevel_1, children [Modulator_2,Modulator_2,Modulator_3,Modulator_3]}
 {
  id=Random_2
  children [WavePlayer_5, WavePlayer_7, WavePlayer_8, WavePlayer_9, WavePlayer_10, WavePlayer_13]
  properties {weights [1,1,1,1,1,1]}
 }
 {id=Random_3, children [WavePlayer_14, WavePlayer_15, WavePlayer_16], properties {weights [1,1,1]}}
 {id=WavePlayer_10, properties {soundWaveAssetPtr=&/Chrome_Assault_Whip_Close_05}}
 {id=WavePlayer_13, properties {soundWaveAssetPtr=&/Chrome_Assault_Whip_Close_06}}
 {id=WavePlayer_14, properties {soundWaveAssetPtr=&/Chrome_Assault_Whip_Close_01}}
 {id=WavePlayer_15, properties {soundWaveAssetPtr=&/Chrome_Assault_Whip_Close_02}}
 {id=WavePlayer_16, properties {soundWaveAssetPtr=&/Chrome_Assault_Whip_Close_03}}
 {id=WavePlayer_5, properties {soundWaveAssetPtr=&Chrome_Assault_Whip_Close_01}}
 {id=WavePlayer_7, properties {soundWaveAssetPtr=&Chrome_Assault_Whip_Close_02}}
 {id=WavePlayer_8, properties {soundWaveAssetPtr=&Chrome_Assault_Whip_Close_03}}
 {id=WavePlayer_9, properties {soundWaveAssetPtr=&/Chrome_Assault_Whip_Close_04}}
]
properties {
 bOverrideAttenuation=true
 maxDistance=2097152
 soundClassObject=&Fort_WeaponRanged_Projectile_Effects
 volumeMultiplier=0.6
 attenuationOverrides {bAttenuate=false, stereoSpread=0}
 concurrencySet [/Game/Sounds/Concurrency_Rules/Athena_BulletWhizBy_Conc]
}
