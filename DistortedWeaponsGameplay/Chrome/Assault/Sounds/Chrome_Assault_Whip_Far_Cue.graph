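# Sound cue for the distant bullet whip (whiz-by) of the Chrome Assault weapon, classed as a ranged projectile effect.
# A quality-level node selects either a six-wave or a two-wave equally weighted random pool, each serving two levels.
# A root modulator varies pitch upward and boosts volume, offset by a halved cue volume.
# Attenuation override turns off distance falloff and stereo spread, and the shared Athena bullet whiz-by concurrency rule governs simultaneous plays.
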SoundCue /DistortedWeaponsGameplay/Chrome/Assault/Sounds/Chrome_Assault_Whip_Far_Cue
root=Modulator_1
nodes [
 {id=Modulator_1, children [QualityLevel_1], properties {pitchMax=1.2, pitchMin=1, volumeMax=1.9, volumeMin=1.3}}
 {id=QualityLevel_1, children [Random_4,Random_4,Random_5,Random_5]}
 {
  id=Random_4
  children [WavePlayer_16, WavePlayer_17, WavePlayer_18, WavePlayer_19, WavePlayer_20, WavePlayer_21]
  properties {weights [1,1,1,1,1,1]}
 }
 {id=Random_5, children [WavePlayer_22, WavePlayer_23], properties {weights [1,1]}}
 {id=WavePlayer_16, properties {soundWaveAssetPtr=&/Chrome_Assault_Whip_Far_01}}
 {id=WavePlayer_17, properties {soundWaveAssetPtr=&/Chrome_Assault_Whip_Far_02}}
 {id=WavePlayer_18, properties {soundWaveAssetPtr=&/Chrome_Assault_Whip_Far_03}}
 {id=WavePlayer_19, properties {soundWaveAssetPtr=&/Chrome_Assault_Whip_Far_04}}
 {id=WavePlayer_20, properties {soundWaveAssetPtr=&/Chrome_Assault_Whip_Far_05}}
 {id=WavePlayer_21, properties {soundWaveAssetPtr=&/Chrome_Assault_Whip_Far_06}}
 {id=WavePlayer_22, properties {soundWaveAssetPtr=&Chrome_Assault_Whip_Far_01}}
 {id=WavePlayer_23, properties {soundWaveAssetPtr=&Chrome_Assault_Whip_Far_02}}
]
properties {
 bOverrideAttenuation=true
 maxDistance=2097152
 soundClassObject=&Fort_WeaponRanged_Projectile_Effects
 volumeMultiplier=0.5
 attenuationOverrides {bAttenuate=false, stereoSpread=0}
 concurrencySet [/Game/Sounds/Concurrency_Rules/Athena_BulletWhizBy_Conc]
}
